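# Deform Control Rig that drives upper-arm twist bones for the PER part of the female medium Grasshopper player body.
# Each forward step extracts the X-axis twist from upperarm_l/r local rotation via swing-twist decomposition.
# The twist angle, scaled by shared negative constants, partially counter-rotates each side's two twist bones, more strongly on twist_01, propagating to children.
# The listed hierarchy omits the upper-arm and twist bones the graph references.
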ControlRigBlueprint /BRCosmetics/Characters/Player/Female/Medium/Bodies/F_MED_Grasshopper/Meshes/Parts/PER/F_MED_Grasshopper_PER_Deform_CtrlRig_twists
graphs [
 RigVMFunctionLibrary {links [], nodes []}
 RigVMModel {nodes [
   RigUnit_BeginExecution = &RigUnit_BeginExecution()
   n1 = /Script/RigVM.RigVMFunction_MathQuaternionToAxisAndAngle(Value=&RigVMFunction_MathQuaternionSwingTwist(TwistAxis=(X=1,Y=0,Z=0), Input=&RigUnit_GetBoneTransform(Bone=upperarm_l, Space=LocalSpace).Transform.Rotation).Twist) {is=struct RigVMUnitNode, template="ToAxisAndAngle::Execute(in Value,out Axis,out Angle)"}
   RerouteNode_2_1 = RigVMDispatchNode(Value=-0.8) {template="DISPATCH_RigVMDispatch_Constant(io Value)"}
   /Script/ControlRig.RigUnit_SetBoneRotation(
    Bone=upperarm_twist_01_l
    Weight=1
    bPropagateToChildren=True
    Rotation=/Script/RigVM.RigVMFunction_MathQuaternionFromAxisAndAngle(
     Angle=&RigVMFunction_MathFloatMul(A=@n1.Angle, B=@RerouteNode_2_1.Value) {template="Multiply::Execute(in A,in B,out Result)"}.Result
     Axis=@n1.Axis
    ) {is=struct RigVMUnitNode, template="FromAxisAndAngle::Execute(in Axis,in Angle,out Result)"}.Result
    ExecutePin=@RigUnit_BeginExecution.ExecutePin
   ) {is=struct RigVMUnitNode}
   n2 = &RigVMFunction_MathQuaternionToAxisAndAngle(Value=&RigVMFunction_MathQuaternionSwingTwist(TwistAxis=(X=1,Y=0,Z=0), Input=&RigUnit_GetBoneTransform(Bone=upperarm_l, Space=LocalSpace).Transform.Rotation).Twist) {template="ToAxisAndAngle::Execute(in Value,out Axis,out Angle)"}
   RerouteNode_2 = RigVMDispatchNode(Value=-0.5) {template="DISPATCH_RigVMDispatch_Constant(io Value)"}
   > &RigUnit_SetBoneRotation(
    Bone=upperarm_twist_02_l
    Weight=1
    bPropagateToChildren=True
    Rotation=&RigVMFunction_MathQuaternionFromAxisAndAngle(
     Angle=&RigVMFunction_MathFloatMul(A=@n2.Angle, B=@RerouteNode_2.Value) {template="Multiply::Execute(in A,in B,out Result)"}.Result
     Axis=@n2.Axis
    ) {template="FromAxisAndAngle::Execute(in Axis,in Angle,out Result)"}.Result
   )
   n3 = &RigVMFunction_MathQuaternionToAxisAndAngle(Value=&RigVMFunction_MathQuaternionSwingTwist(TwistAxis=(X=1,Y=0,Z=0), Input=&RigUnit_GetBoneTransform(Bone=upperarm_r, Space=LocalSpace).Transform.Rotation).Twist) {template="ToAxisAndAngle::Execute(in Value,out Axis,out Angle)"}
   > &RigUnit_SetBoneRotation(
    Bone=upperarm_twist_01_r
    Weight=1
    bPropagateToChildren=True
    Rotation=&RigVMFunction_MathQuaternionFromAxisAndAngle(
     Angle=&RigVMFunction_MathFloatMul(A=@n3.Angle, B=@RerouteNode_2_1.Value) {template="Multiply::Execute(in A,in B,out Result)"}.Result
     Axis=@n3.Axis
    ) {template="FromAxisAndAngle::Execute(in Axis,in Angle,out Result)"}.Result
   )
   n4 = &RigVMFunction_MathQuaternionToAxisAndAngle(Value=&RigVMFunction_MathQuaternionSwingTwist(TwistAxis=(X=1,Y=0,Z=0), Input=&RigUnit_GetBoneTransform(Bone=upperarm_r, Space=LocalSpace).Transform.Rotation).Twist) {template="ToAxisAndAngle::Execute(in Value,out Axis,out Angle)"}
   > &RigUnit_SetBoneRotation(
    Bone=upperarm_twist_02_r
    Weight=1
    bPropagateToChildren=True
    Rotation=&RigVMFunction_MathQuaternionFromAxisAndAngle(
     Angle=&RigVMFunction_MathFloatMul(A=@n4.Angle, B=@RerouteNode_2.Value) {template="Multiply::Execute(in A,in B,out Result)"}.Result
     Axis=@n4.Axis
    ) {template="FromAxisAndAngle::Execute(in Axis,in Angle,out Result)"}.Result
   )
  ]}
]
hierarchy [
 root {loc=<0 0 0>, rot=<0 0 0 1>}
 pelvis {parent=root, loc=<0 -0.5301 98.0571>, rot=<0.0127 0.707 -0.0127 -0.707>}
 spine_01 {parent=pelvis, loc=<6.1208 -0.9328 0>, rot=<0 0 0.034 -0.9994>}
 spine_02 {parent=spine_01, loc=<6.5996 -0.2095 0>, rot=<0 0 0.0238 -0.9997>}
 spine_03 {parent=spine_02, loc=<6.4914 -0.1237 0>, rot=<0 0 0.0149 -0.9999>}
 spine_04 {parent=spine_03, loc=<6.9598 0.0794 0>, rot=<0 0 -0.0132 -0.9999>}
 spine_05 {parent=spine_04, loc=<6.9987 0 0>, rot=<0 0 -0.0015 -1>}
 clavicle_l {parent=spine_05, loc=<5.4831 -0.1053 -3.2946>, rot=<0.0465 -0.7289 -0.009 -0.6829>}
 deform_pectoral_fr_l {parent=clavicle_l, loc=<6.2796 0 0>, rot=<0 0 0 1>}
 clavicle_r {parent=spine_05, loc=<5.4831 -0.1053 3.2946>, rot=<0.7289 0.0465 -0.6829 0.009>}
 deform_pectoral_fr_r {parent=clavicle_r, loc=<-6.28 0 0>, rot=<0 0 0 1>}
 neck_01 {parent=spine_05, loc=<11.3138 0.1529 0>, rot=<0 0 -0.2003 -0.9797>}
 neck_02 {parent=neck_01, loc=<4.5627 0 0>, rot=<0 0 0.0504 -0.9987>}
 head {parent=neck_02, loc=<5.3914 0 0>, rot=<0 0 0.115 -0.9934>}
 faceAttach {parent=head, loc=<-3.0697 -0.1623 0.0114>, rot=<0.0082 -0.7071 -0.0082 -0.7071>, scale=<0.89 0.89 0.89>}
 C_jaw {parent=faceAttach, loc=<0.0699 3.0221 1.2803>, rot=<0 0 0 1>}
 Dyn_Master_Hair {parent=head, loc=<5.1645 5.7632 -5.9554>, rot=<-0.0002 0.005 0.9995 -0.0317>}
 Dyn_Hair1 {parent=Dyn_Master_Hair, loc=<3.6141 0.4988 -2.3834>, rot=<-0.428 -0.7538 0.2908 -0.405>}
 Dyn_Hair2 {parent=Dyn_Master_Hair, loc=<5.2611 2.336 -2.6333>, rot=<-0.5384 -0.6361 0.1541 -0.5308>}
 Dyn_Hair3 {parent=Dyn_Master_Hair, loc=<-6.5189 -1.0896 0>, rot=<-0.5925 -0.7076 0.2951 -0.2471>}
 Dyn_Master_Antenna_l {parent=head, loc=<12.9838 2.0403 5.1143>, rot=<-0.0066 0.2022 -0.0321 -0.9788>}
 Dyn_Antenna_l1 {parent=Dyn_Master_Antenna_l, loc=<0.5391 0 0>, rot=<0 0 0 1>}
 Dyn_Antenna_l2 {parent=Dyn_Antenna_l1, loc=<2.9096 0 0>, rot=<0.0001 -0.0157 -0.0069 -0.9999>}
 Dyn_Antenna_l3 {parent=Dyn_Antenna_l2, loc=<5.3545 0 0>, rot=<0.0006 -0.0217 -0.0284 -0.9994>}
 Dyn_Antenna_l4 {parent=Dyn_Antenna_l3, loc=<5.8121 0 -0.017>, rot=<0 -0.1474 0 -0.9891>}
 Dyn_Master_Antenna_r {parent=head, loc=<12.9618 2.0389 -5.1318>, rot=<0.0064 -0.1947 -0.0322 -0.9803>}
 Dyn_Antenna_r1 {parent=Dyn_Master_Antenna_r, loc=<0.5391 0 0>, rot=<0 0 0 1>}
 Dyn_Antenna_r2 {parent=Dyn_Antenna_r1, loc=<2.9096 0 0>, rot=<0 0.0056 -0.0069 -1>}
 Dyn_Antenna_r3 {parent=Dyn_Antenna_r2, loc=<5.3545 0 0>, rot=<-0.0004 0.0157 -0.0284 -0.9995>}
 Dyn_Antenna_r4 {parent=Dyn_Antenna_r3, loc=<5.8121 0 -0.017>, rot=<0 0.1592 0 -0.9872>}
 Dyn_Master_Flap {parent=spine_05, loc=<5.8108 -8.2044 0>, rot=<0.0395 -0.9992 0 0>}
 Dyn_Master_Flap_l {parent=Dyn_Master_Flap, loc=<0 0 -4.9383>, rot=<0 0 0 -1>}
 Dyn_Flap_l1 {parent=Dyn_Master_Flap_l, loc=<0.2279 0 0>, rot=<0 0 0 -1>}
 Dyn_Flap_l2 {parent=Dyn_Flap_l1, loc=<3.8628 0 0>, rot=<0 0 0 -1>}
 Dyn_Flap_l3 {parent=Dyn_Flap_l2, loc=<6.2682 0 0>, rot=<0 0 -0.0076 -1>}
 Dyn_Flap_l4 {parent=Dyn_Flap_l3, loc=<4.4726 0 0>, rot=<0 0 0 1>}
 Dyn_Master_Flap_r {parent=Dyn_Master_Flap, loc=<0 0 4.9189>, rot=<0 0 0 -1>}
 Dyn_Flap_r1 {parent=Dyn_Master_Flap_r, loc=<0.2279 0 0>, rot=<0 0 0 -1>}
 Dyn_Flap_r2 {parent=Dyn_Flap_r1, loc=<3.8628 0 0>, rot=<0 0 0 -1>}
 Dyn_Flap_r3 {parent=Dyn_Flap_r2, loc=<6.2682 0 0>, rot=<0 0 -0.0076 -1>}
 Dyn_Flap_r4 {parent=Dyn_Flap_r3, loc=<4.4726 0 0>, rot=<0 0 0 1>}
 attach {parent=root, loc=<0 0 0>, rot=<0 0 0 1>}
 dyn_root_simspace {parent=root, loc=<0 0 -2.1558>, rot=<0 0 0.401 -0.9161>}
 dyn_root_slope {parent=root, loc=<0 0.0001 -2.1567>, rot=<0 0 0 1>}
]
